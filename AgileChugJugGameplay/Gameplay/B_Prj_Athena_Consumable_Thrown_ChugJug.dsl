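; Thrown Chug Jug projectile that turns back into a world pickup when it lands, keeping the item's current ammo count.
; SetItem stores the thrown item and copies a drop entry created from it into the FortItemEntry component.
; On bounce or SpawnItem, the server spawns a tossed pickup from that entry at an offset, runs spawn hooks and FX, and sets the owning controller as dropper.
; Both events duplicate this logic, and SpawnItem ignores its water-hit flag.
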(event EventBeginPlay)

(event EventOnBounce
  (params (Hit :type Hit%20Result%20Structure%20%28by%20ref%29))
  ; On Hit Spawn Pickup using ItemEntry to keep current ammo count
  (Networking|SwitchHasAuthority
    (:Authority
      (bind _optional_owner_pc (CallFunction|GetSpawnedPickupOptionalOwnerPC))
      (bind _returnvalue (Inventory|SpawnPickupInWorldWithClassAndItemEntry (Class|FortItemEntryComponent|GetOwnedItem (Variables|Default|GetFortItemEntry)) "/Script/FortniteGame.FortPickupAthena" (+ (Transformation|GetActorLocation) (Variables|Default|GetSpawnPositionOffset)) (Transformation|GetActorForwardVector) 0 true (Variables|Default|GetShouldUseRandomRotationWhenSpawningPickUp) false "Tossed" "Unset" _optional_owner_pc false))
      (CallFunction|OnPickupSpawned _returnvalue)
      (CallFunction|SpawnedItemFx)
      (bind _returnvalue_2 (PlayerInfo|GetOwnerPlayerController))
      (Class|FortPickup|SetDropper _returnvalue _returnvalue_2))))

(event EventSpawnItem
  (params (bHitWater :type Boolean))
  ; On Hit Spawn Pickup using ItemEntry to keep current ammo count
  (Networking|SwitchHasAuthority
    (:Authority
      (bind _optional_owner_pc (CallFunction|GetSpawnedPickupOptionalOwnerPC))
      (bind _returnvalue (Inventory|SpawnPickupInWorldWithClassAndItemEntry (Class|FortItemEntryComponent|GetOwnedItem (Variables|Default|GetFortItemEntry)) "/Script/FortniteGame.FortPickupAthena" (+ (Transformation|GetActorLocation) (Variables|Default|GetSpawnPositionOffset)) (Transformation|GetActorForwardVector) 0 true (Variables|Default|GetShouldUseRandomRotationWhenSpawningPickUp) false "Tossed" "Unset" _optional_owner_pc false))
      (CallFunction|OnPickupSpawned _returnvalue)
      (CallFunction|SpawnedItemFx)
      (bind _returnvalue_2 (PlayerInfo|GetOwnerPlayerController))
      (Class|FortPickup|SetDropper _returnvalue _returnvalue_2))))

(fn SetItem
  (params (Item :type Fort%20Item%20Object%20Reference))
  (bind _output_get (Variables|Default|SetItem Item))
  (CallFunction|SetitemEntryonDrop _output_get))

(fn SetitemEntryOnDrop
  (params (ItemObj :type Fort%20Item%20Object%20Reference))
  (Utilities|Casting|CastToFortPlayerPawn (Game|GetInstigator))
  (bind _asfort_world_item (Utilities|Casting|CastToFortWorldItem ItemObj))
  (Class|FortItemEntryComponent|SetOwnedItem (Variables|Default|GetFortItemEntry) (Item|CreateItemEntryforDrop _asfort_world_item)))

(fn ConstructionScript
  (|CallParentFunction))
