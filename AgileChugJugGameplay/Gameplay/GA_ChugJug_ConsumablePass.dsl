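; Gameplay ability for throwing a Chug Jug, handing the held inventory item over to the thrown projectile.
; On activation it captures the source weapon's inventory item, runs the parent logic, then waits for the ItemThrown gameplay event on the avatar.
; Where cosmetic events cannot execute, it sets that item on the Chug Jug projectile and removes one from the player's inventory.
; An invalid projectile reference only prints a debug message.
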(event Ability|EventOnEndAbility
  (params (bWasCancelled :type Boolean))
  (|CallParentFunction bWasCancelled))

(event Ability|EventActivateAbility
  (bind _returnvalue (Ability|GetCurrentSourceWeapon))
  (bind _returnvalue_2 (FortWeapon|GetInventoryItem _returnvalue))
  ; Comment
  (|CallParentFunction)
  (Ability|Async|WaitGameplayEventToActor (Ability|GetAvatarActorfromActorInfo) "(TagName=\"Event.Athena.Item.ItemThrown\")"
    (:EventReceived
      (Networking|CanExecuteCosmeticEvents
        (:False
          (bind _projectilereference (Variables|ProjectileProperties|GetProjectileReference)
            (:then
              (bind _asb_prj_athena_consumable_thrown_chug_jug (Utilities|Casting|CastToB_Prj_Athena_Consumable_Thrown_ChugJug _projectilereference))
              (Class|BPrjAthenaConsumableThrownChugJug|SetItem _asb_prj_athena_consumable_thrown_chug_jug _returnvalue_2 :!class /AgileChugJugGameplay/Gameplay/B_Prj_Athena_Consumable_Thrown_ChugJug.B_Prj_Athena_Consumable_Thrown_ChugJug_C)
              (Inventory|RemoveFortItemFromPlayer (Fortnite|GetFortPlayerControllerfromActor (Ability|GetActivatingPawn)) _returnvalue_2 1 true))
            (:else
              (Development|PrintString "THROWN CHUG JUG NOT VALID" true true "(R=1.000000,G=0.042746,B=0.000000,A=1.000000)" 5.0))))))))
